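# Landscape material function that blends one grass texture between near and far tiling scales to vary its look with distance.
# World XY, excluding shader offsets, is divided by two scale parameters to form UVs, optionally through vertex interpolators via a static switch.
# Both samples share the Grass texture parameter and are lerped by MF_TerrainDistanceFade.
# It outputs blended RGB and alpha, plus the far-scale RGB separately.
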MaterialFunction /FigmentEnvironment/Migration51/Athena/Environments/Landscape/MaterialFunctions/MF_Landscape_GrassDistanceBlend
graph {nodes [
  n1 = ComponentMask(None=WorldPosition() {worldPositionShaderOffset=WPT_ExcludeAllShaderOffsets}.XYZ) {g=true, r=true}
  n2 = Divide(A=@n1, B=ScalarParameter() {defaultValue=700, parameterName=GrassNearScale})
  n3 = FunctionInput(Preview=StaticBool()) {bUsePreviewValueAsDefault=true, inputName=UseVertexInterp, inputType=FunctionInput_StaticBool}
  n4 = TextureSampleParameter2D(UVs=StaticSwitch(True=VertexInterpolator(VS=@n2).PS, False=@n2, Value=@n3)) {
   group=Grass
   parameterName=Grass
   samplerSource=SSM_Wrap_WorldGroupSettings
   texture=/FigmentEnvironment/Migration51/Athena/Environments/Landscape/Textures/T_Athena_Grass_Farm_ColorMatched_D_2
  }
  n5 = Divide(A=@n1, B=ScalarParameter() {defaultValue=3000, parameterName=GrassFarScale})
  n6 = TextureSampleParameter2D(UVs=StaticSwitch(True=VertexInterpolator(VS=@n5).PS, False=@n5, Value=@n3)) {
   group=Grass
   parameterName=Grass
   samplerSource=SSM_Wrap_WorldGroupSettings
   texture=&T_Athena_Grass_Farm_ColorMatched_D_2
  }
  n7 = MaterialFunctionCall() {materialFunction=&/MF_TerrainDistanceFade}
  FunctionOutput(None=LinearInterpolate(A=@n4.RGB, B=@n6.RGB, Alpha=@n7.Result)) {outputName=Grass RGB}
  FunctionOutput(None=LinearInterpolate(A=@n4.A, B=@n6.A, Alpha=@n7.Result)) {outputName=Grass A, sortPriority=1}
  FunctionOutput(None=@n6.RGB) {outputName=Distant Grass RGB, sortPriority=3}
 ]}
